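# Niagara system for the idle effect of the EmberRae pickaxe, showing two purple gem glow sprites on the weapon.
# A single GemGlow emitter bursts two camera-offset sprites that are not killed when their lifetime elapses.
# Each update re-samples the user-supplied skeletal mesh, using particle ID modulo 2 as the bone or socket index so each sprite tracks its own point.
# A time-of-day scalar scales alpha, dimming the glow at night.
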NiagaraSystem /BRCosmetics/Weapons/FORT_Melee/Pickaxe_EmberRae/FX/NS_Pickaxe_EmberRae_Idle
dependencies {
 usedDataInterfaces []
 usedDynamicInputs [&Multiply_Vector2D_ByFloat]
 usedModules [&SystemState, &Timeline]
}
emitters {GemGlow {
  emitterUpdateScript [
   &EmitterState()
   &SpawnBurst_Instantaneous(Spawn Count=2, Use Spawn Probability=false, UseLoopCountLimit=true)
  ]
  particleSpawnScript [
   &InitializeParticle(
    Color=<rgba 0.7754 0 1.5 1>
    Lifetime=0.25
    Sprite Size Mode=::NewEnumerator3
    Uniform Sprite Size=15
    UsePositionOffset=false
    Write Material Random=true
   )
   &CameraOffset(Camera Offset Amount=10)
  ]
  particleUpdateScript [
   &ParticleState(Kill Particles When Lifetime Has Elapsed=false, Let Infinitely Lived Particles Die When Emitter Deactivates=true)
   &SkeletalMeshLocation(
    Bone / Socket Index=&ModuloInt(A=&ReturnParticleIDAttributes(), B=2)
    MaskSpawnGroup=false
    Mesh Sampling Type=::NewEnumerator5
    Skeletal Mesh=$User.Skeletal Mesh
    Socket Sampling Mode=::NewEnumerator5
   )
   &ScaleColor(
    Scale Alpha=&FNTimeOfDayScalar(Night Scale=0.75, Morning Scale=1, Evening Scale=1)
    ScaleA=true
    ScaleRGB=true
   )
  ]
  rendererClasses [&NiagaraSpriteRendererProperties]
  renderers [{rendererClass=&NiagaraSpriteRendererProperties}]
 }}
summary {
 systemName=NS_Pickaxe_EmberRae_Idle
 emitters [{emitterName=GemGlow, rendererClasses [&NiagaraSpriteRendererProperties]}]
 userVariables [User.Skeletal Mesh {
   type=NiagaraDataInterfaceSkeletalMesh
   defaultValue {
    struct=NiagaraExt_VariableValue_DataInterface
    value {dataInterface=&self:NiagaraDataInterfaceSkeletalMesh_0}
   }
  }]
}
