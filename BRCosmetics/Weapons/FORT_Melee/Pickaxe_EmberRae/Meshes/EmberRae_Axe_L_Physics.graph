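# Physics asset for the left Ember Rae pickaxe axe mesh, providing bodies and joints for its dangling strap.
# A capsule covers the two-handed melee handle, with a small sphere on the strap's master bone rotationally locked to it as the anchor.
# Three capsule segments chain from that anchor, with narrower swing limit angles on the two outer joints than at the root.
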PhysicsAsset /BRCosmetics/Weapons/FORT_Melee/Pickaxe_EmberRae/Meshes/EmberRae_Axe_L_Physics
bodies {
 Melee_TwoHanded_Handle [{shapeType=Capsule, length=28.2119, radius=2.9621, center=<0.0748 -0.0258 -3.0123>, rotation=<rot -0.135 -90 -0.5033>}]
 dyn_strap_L_1 [{shapeType=Capsule, length=1.1778, radius=2.0364, shapeName=dyn_strap_L_1_capsule, center=<2.481 0 0>, rotation=<rot 90 180 -90>}]
 dyn_strap_L_2 [{shapeType=Capsule, length=1.2144, radius=1.9765, shapeName=dyn_strap_L_2_capsule, center=<2.4595 0 0>, rotation=<rot 0 0 90>}]
 dyn_strap_L_3 [{shapeType=Capsule, length=1.0902, radius=2.3153, shapeName=dyn_strap_L_3_capsule, center=<3.4888 0 0>, rotation=<rot 0 0 90>}]
 dyn_strap_MSTR_L [{shapeType=Sphere, radius=0.505, shapeName=dyn_strap_MSTR_L_sphere}]
}
constraints [
 {
  bone1Name=dyn_strap_MSTR_L
  bone2Name=Melee_TwoHanded_Handle
  swing1LimitDegrees=90
  swing1Motion=Locked
  swing2LimitDegrees=90
  swing2Motion=Locked
  twistLimitDegrees=5
  twistMotion=Locked
 }
 {bone1Name=dyn_strap_L_1, bone2Name=dyn_strap_MSTR_L, swing1LimitDegrees=90, swing2LimitDegrees=90, twistLimitDegrees=5}
 {bone1Name=dyn_strap_L_2, bone2Name=dyn_strap_L_1, swing1LimitDegrees=25, swing2LimitDegrees=25, twistLimitDegrees=5}
 {bone1Name=dyn_strap_L_3, bone2Name=dyn_strap_L_2, swing1LimitDegrees=25, swing2LimitDegrees=25, twistLimitDegrees=5}
]
