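# Physics asset for the right-hand axe mesh of the EmberRae pickaxe, giving its dangling strap ragdoll-style secondary motion.
# A capsule covers the handle, and a small sphere on the strap master bone is locked to it in swing and twist as the anchor.
# Three strap capsules chain from that anchor, with a wide swing at the first joint, tighter swing further down and minimal twist throughout.
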PhysicsAsset /BRCosmetics/Weapons/FORT_Melee/Pickaxe_EmberRae/Meshes/EmberRae_Axe_R_Physics
bodies {
 Melee_TwoHanded_Handle [{shapeType=Capsule, length=34.7733, radius=4.2023, center=<-0.0746 0.0212 -14.9824>, rotation=<rot -0.0268 -90 -0.7628>}]
 dyn_strap_MSTR_R [{shapeType=Sphere, radius=0.505, shapeName=dyn_strap_MSTR_R_sphere}]
 dyn_strap_R_1 [{shapeType=Capsule, length=1.3238, radius=2.7322, shapeName=dyn_strap_R_1_capsule, center=<1.4583 0 0>, rotation=<rot 0 0 -90>}]
 dyn_strap_R_2 [{shapeType=Capsule, length=1.2611, radius=2.4961, shapeName=dyn_strap_R_2_capsule, center=<2.8307 -0.0462 -0.0753>, rotation=<rot 0 0 -90>}]
 dyn_strap_R_3 [{shapeType=Capsule, length=1.6346, radius=2.378, shapeName=dyn_strap_R_3_capsule, center=<2.828 0.2205 0.0464>, rotation=<rot 0 0 -90>}]
}
constraints [
 {bone1Name=dyn_strap_R_1, bone2Name=dyn_strap_MSTR_R, swing1LimitDegrees=90, swing2LimitDegrees=90, twistLimitDegrees=5}
 {
  bone1Name=dyn_strap_MSTR_R
  bone2Name=Melee_TwoHanded_Handle
  swing1LimitDegrees=45
  swing1Motion=Locked
  swing2LimitDegrees=45
  swing2Motion=Locked
  twistLimitDegrees=45
  twistMotion=Locked
 }
 {bone1Name=dyn_strap_R_2, bone2Name=dyn_strap_R_1, swing1LimitDegrees=25, swing2LimitDegrees=25, twistLimitDegrees=5}
 {bone1Name=dyn_strap_R_3, bone2Name=dyn_strap_R_2, swing1LimitDegrees=25, swing2LimitDegrees=25, twistLimitDegrees=5}
]
