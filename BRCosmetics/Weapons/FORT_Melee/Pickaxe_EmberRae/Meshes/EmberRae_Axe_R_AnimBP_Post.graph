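# Post-process AnimBlueprint for the right Ember Rae pickaxe that adds physics motion and corrective deformation to its strap.
# The input pose runs through a RigidBody simulation with damped component motion and overridden gravity, then three Modify Bone nodes.
# Each update measures dyn_strap_R_1's Y rotation relative to the master strap bone.
# That angle offsets the master bone and blends in Z squash on R_1, while R_2's scale is reset to one.
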AnimBlueprint /BRCosmetics/Weapons/FORT_Melee/Pickaxe_EmberRae/Meshes/EmberRae_Axe_R_AnimBP_Post
target_skeleton=&/EmberRae_Axe_Skeleton
graphs [
 AnimGraph {nodes [
   AnimGraphNode_SubInput_0 = Input Pose()
   n1 = RigidBody(ComponentPose=Local To Component(LocalPose=@AnimGraphNode_SubInput_0)) {
    bOverrideWorldGravity=true
    overridePhysicsAsset=&/EmberRae_Axe_R_Physics
    alphaBoolBlend {blendOption=Linear}
    componentAppliedLinearAccClamp=<3000 3000 3500>
    componentLinearAccScale=<0.32 0.32 0.4>
    componentLinearVelScale=<0.32 0.32 0.4>
    overrideWorldGravity {z=-980}
   }
   n2 = "Transform (Modify) Bone - Bone: dyn_strap_MSTR_R"(ComponentPose=@n1, Alpha=1, Translation=Get TranslateMSTR() -> (TranslateMSTR:Vector), Rotation="0,0,0", Scale="1,1,1") {
    scaleSpace=BCS_BoneSpace
    translationMode=BMM_Additive
    translationSpace=BCS_BoneSpace
    alphaBoolBlend {blendOption=Linear}
    boneToModify {boneName=dyn_strap_MSTR_R}
    translation {x=-1.639, z=1.629}
   }
   n3 = Get StrapScaleAlpha() -> (StrapScaleAlpha:double)
   n4 = "Transform (Modify) Bone - Bone: dyn_strap_R_1"(ComponentPose=@n2, Alpha=@n3, Translation="0,0,0", Rotation="0,0,0", Scale="1,1,0.7") {
    scaleMode=BMM_Replace
    translationSpace=BCS_BoneSpace
    alphaBoolBlend {blendOption=Linear}
    boneToModify {boneName=dyn_strap_R_1}
    scale {z=0.7}
   }
   n5 = "Transform (Modify) Bone - Bone: dyn_strap_R_2"(ComponentPose=@n4, Alpha=@n3, Translation="0,0,0", Rotation="0,0,0", Scale="1,1,1") {
    rotationMode=BMM_Additive
    scaleMode=BMM_Replace
    translationSpace=BCS_BoneSpace
    alphaBoolBlend {blendOption=Linear}
    boneToModify {boneName=dyn_strap_R_2}
   }
   Output Pose(Result=Component To Local(ComponentPose=@n5)) {layerGroup=None}
  ]}
 EventGraph {nodes [
   Event BlueprintUpdateAnimation()
   n1 = GetAlphaFromReferenceTransformDelta(
    SkeletalMeshComponent=Get SkeletalMesh() -> (SkeletalMesh:Skeletal Mesh Component Object Reference)
    BoneName=dyn_strap_R_1
    BaseName=dyn_strap_MSTR_R
    DeltaComponentType=EulerY
    InputRange=(X=0,Y=1)
    TwistAxis="1,0,0"
   )
   n2 = MapRangeClamped(Value=@n1.RawOutput, InRangeA=0, InRangeB=90, OutRangeB=-1.65)
   n3 = MapRangeClamped(Value=@n1.RawOutput, InRangeA=0, InRangeB=-90, OutRangeB=-1.65)
   > Set TranslateMSTR(TranslateMSTR:Vector=MakeVector(X=float + float(A=@n2, B=@n3), Y=0, Z=MapRangeClamped(Value=@n1.RawOutput, InRangeA=-90, InRangeB=90, OutRangeA=-1.7, OutRangeB=1.7))) -> (Output_Get:Vector)
   > Set StrapScaleAlpha(StrapScaleAlpha:double=float + float(A=MapRangeClamped(Value=@n1.RawOutput, InRangeA=0, InRangeB=90), B=MapRangeClamped(Value=@n1.RawOutput, InRangeA=0, InRangeB=-90))) -> (Output_Get:double)
   Event BlueprintInitializeAnimation()
   > Set SkeletalMesh(SkeletalMesh:Skeletal Mesh Component Object Reference=GetPartSkeletalMeshComponent()) -> (Output_Get:Skeletal Mesh Component Object Reference)
   float + float(A=MapRangeClamped(Value=0, InRangeA=0, InRangeB=90, OutRangeB=1.6), B=MapRangeClamped(Value=0, InRangeA=0, InRangeB=-90, OutRangeB=-1.6))
  ]}
]
variables [
 SkeletalMesh {type=SkeletalMeshComponent}
 StrapScaleAlpha {type=number}
 TranslateMSTR {type=Vector}
]
